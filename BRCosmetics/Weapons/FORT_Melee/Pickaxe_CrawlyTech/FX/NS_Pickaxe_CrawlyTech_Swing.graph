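# Niagara swing trail effect for the CrawlyTech pickaxe, built from three ribbon anim-trail emitters: main, tip and lightning.
# Each picks its socket pair from User.SwingSectionIndex, so the ribbons follow the current swing section.
# Main and tip trails take tint and opacity from system-level time-of-day and fade values, while the lightning trail uses its own HDR blue curves and dynamic material parameters.
# A shared System.Kill flag clears all particles.
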NiagaraSystem /BRCosmetics/Weapons/FORT_Melee/Pickaxe_CrawlyTech/FX/NS_Pickaxe_CrawlyTech_Swing
dependencies {
 usedDataInterfaces [&NiagaraDataInterfaceColorCurve, &NiagaraDataInterfaceCurve, &NiagaraDataInterfaceSocketReader]
 usedDynamicInputs [
  &MergeHelper_FloatSwitch
  &Add_Float
  &SetBoolByFloatComparison
 ]
 usedModules [
  &self:SystemSpawnScript.SystemScriptSource.SystemScriptGraph.NiagaraNodeAssignment_0.SetVariables_#1
  &SystemState
  &self:SystemSpawnScript.SystemScriptSource.SystemScriptGraph.NiagaraNodeAssignment_2.SetVariables_#2
 ]
}
emitters {
 LightningTrail {
  emitterUpdateScript [
   &EmitterState()
   &FN_SpawnAnimTrail(Spawn Interval=0) {module=FN_SpawnAnimTrail001}
  ]
  particleSpawnScript [
   &FN_InitializeAnimTrailBoneIndecies(
    First Bone or Socket Index:NiagaraInt32=&MergeHelper_IntSwitch(0=2, 1=0, 2=2, 3=2, 4=?, 5=?, Index=$User.SwingSectionIndex, Num=::NewEnumerator2)
    Second Bone or Socket Index:NiagaraInt32=&MergeHelper_IntSwitch(0=3, 1=3, 2=1, 3=1, 4=?, 5=?, Index=$User.SwingSectionIndex, Num=::NewEnumerator2)
   )
   &FN_InitializeAnimTrail(AnimTrailWidth=0.9, Lifetime=0.15)
  ]
  particleUpdateScript [
   &KillParticles(Kill Particles=$System.Kill)
   &ScaleColor(
    Scale Alpha=&FloatFromCurve(FloatCurve=DataInterface() {propertyValues {Curve {keys [
         {arriveTangent=-7.1431, interpMode=RCIM_Cubic, leaveTangent=-7.1431, tangentMode=RCTM_User, time=0.2, value=1}
         {arriveTangent=0.1174, tangentMode=RCTM_User, time=0.6, value=0}
        ]}}})
    Scale RGB=&Multiply_VectorByFloat(
     Vector=&MakeVectorFromLinearColorRGB(Color=&ColorFromCurve(ColorCurve=DataInterface() {propertyValues {
         AlphaCurve {keys []}
         BlueCurve {keys [<at 0.1797 10 RCIM_Cubic>,<at 0.9993 1 RCIM_Cubic>]}
         GreenCurve {keys [<at 0.1797 4 RCIM_Cubic>,<at 0.9993 0.25 RCIM_Cubic>]}
         RedCurve {keys [<at 0.1797 1 RCIM_Cubic>,<at 0.9993 0.1 RCIM_Cubic>]}
        }}))
     Float=&Multiply_Float(A=1, B=&FNTimeOfDayScalar(Night Scale=0.8, Morning Scale=1, Evening Scale=1))
    )
    ScaleA=true
    ScaleRGB=true
   )
   &DynamicMaterialParameters(
    Index 0 Param 1=2
    Index 0 Param 2=&FloatFromCurve(FloatCurve=DataInterface() {propertyValues {Curve {keys [
         {arriveTangent=-0.1402, interpMode=RCIM_Cubic, leaveTangent=-0.1402, tangentMode=RCTM_User, time=0, value=0.05}
         {interpMode=RCIM_Cubic, tangentMode=RCTM_User, time=1, value=0}
        ]}}})
    Index 0 Param 3=10
    Module.Param0WriteEnabled=true
    Module.Param0WriteEnabled1=true
    Module.Param1WriteEnabled=true
    Module.Param1WriteEnabled1=true
    Module.Param2WriteEnabled=false
    Module.Param2WriteEnabled1=true
    Module.Param3WriteEnabled=false
    Module.Param3WriteEnabled1=true
    Write Parameter Index 1=true
   )
   &FN_UpdateAnimTrail() {module=FN_UpdateAnimTrail001}
   /SetVariables(Particles.TrailWidth=20) {isSetParameters=true}
   /SetVariables(Particles.DY2:Vector4f=&MakeVector4(X=1, Y=1, Z=1, W=1)) {isSetParameters=true}
   /SetVariables(Particles.Trail Color02:LinearColor=&MakeLinearColorFromVectorAndFloat(
     "Vector (RGB)"=&MakeVectorFromLinearColorRGB(Color=&MultiplyLinearColor(
       Linear Color=<rgba 0.375 0.7292 2.5 1>
       Scale Factor=&MakeVector4_FromLinearColor(LinearColor=&ColorFromCurve(ColorCurve=DataInterface() {propertyValues {
           AlphaCurve {keys []}
           BlueCurve {keys [<at 0 1.5 RCIM_Cubic>,<at 0.749 0.8>]}
           GreenCurve {keys [<at 0 1.5 RCIM_Cubic>,<at 0.749 0.8>]}
           RedCurve {keys [<at 0 1.5 RCIM_Cubic>,<at 0.749 0.8>]}
          }}))
      ))
     "Float (Alpha)"=&Multiply_Float(
      A=&FNTimeOfDayScalar(Night Scale=0.5, Morning Scale=1, Evening Scale=1)
      B=&FloatFromCurve(
       FloatCurve=DataInterface() {propertyValues {Curve {keys [
           {arriveTangent=-7.1431, interpMode=RCIM_Cubic, leaveTangent=-7.1431, tangentMode=RCTM_User, time=0.2, value=1}
           {arriveTangent=0.1174, tangentMode=RCTM_User, time=0.6, value=0}
          ]}}}
       Scale Curve=&FN_FrontEndFloatModifier(Gameplay=0.5, Front End=0.5)
      )
     )
    )) {isSetParameters=true}
   /SetVariables(Particles.TrailWidthCenter=17) {isSetParameters=true}
   &SolveForcesAndVelocity(Clamp Velocity=false, Limit Acceleration=false)
  ]
  rendererClasses [&NiagaraRibbonRendererProperties]
  renderers [
   {rendererClass=&NiagaraRibbonRendererProperties}
   {rendererClass=&NiagaraRibbonRendererProperties, rendererIndex=1}
  ]
 }
 MainTrail {
  emitterUpdateScript [
   &EmitterState()
   &FN_SpawnAnimTrail(AnimTrailWidthMode=::FromSecond, Spawn Interval=0) {module=FN_SpawnAnimTrail001}
  ]
  particleSpawnScript [
   &FN_InitializeAnimTrailBoneIndecies(
    First Bone or Socket Index:NiagaraInt32=&MergeHelper_IntSwitch(0=2, 1=0, 2=2, 3=2, 4=?, 5=?, Index=$User.SwingSectionIndex, Num=::NewEnumerator2)
    Second Bone or Socket Index:NiagaraInt32=&MergeHelper_IntSwitch(0=3, 1=3, 2=1, 3=1, 4=?, 5=?, Index=$User.SwingSectionIndex, Num=::NewEnumerator2)
   )
   &FN_InitializeAnimTrail(
    Color=&MultiplyLinearColor(
     Linear Color=$System.TOD
     Scale Factor=&MakeFloat4FromVectorAndFloat(
      "Vector (RGB)"=&MakeVectorFromLinearColorRGB(Color=<rgba 0.475 0.5 0.4855 1>)
      "Float (Alpha)"=$System.TrailFade
     )
    )
    Lifetime=0.18
   )
  ]
  particleUpdateScript [
   &KillParticles(Kill Particles=$System.Kill)
   &ScaleColor(
    Scale Alpha=&FloatFromCurve(
     FloatCurve=DataInterface() {propertyValues {Curve {keys [{arriveTangent=-2.7426, interpMode=RCIM_Cubic, leaveTangent=-2.7426, tangentMode=RCTM_User, time=0, value=1}, <at 1 0>]}}}
     Scale Curve=3
    )
    Scale RGB=&Multiply_Vector(
     A=&MakeVectorFromLinearColorRGB(Color=&ColorFromCurve(ColorCurve=DataInterface() {propertyValues {
         AlphaCurve {keys [<at 0 1 RCIM_Cubic>]}
         BlueCurve {keys [<at 0.4018 1>]}
         GreenCurve {keys [<at 0.4018 1>]}
         RedCurve {keys [<at 0.4018 1>]}
        }}))
     B=&MakeVectorFromLinearColorRGB(Color=$System.TOD)
    )
    ScaleA=true
    ScaleRGB=true
   )
   &FN_UpdateAnimTrail() {module=FN_UpdateAnimTrail001}
   &SolveForcesAndVelocity(Clamp Velocity=false, Limit Acceleration=false)
  ]
  rendererClasses [&NiagaraRibbonRendererProperties]
  renderers [{rendererClass=&NiagaraRibbonRendererProperties}]
 }
 TipTrail {
  emitterUpdateScript [
   &EmitterState()
   &FN_SpawnAnimTrail(AnimTrailWidthMode=::FromSecond, Spawn Interval=0) {module=FN_SpawnAnimTrail001}
  ]
  particleSpawnScript [
   &FN_InitializeAnimTrailBoneIndecies(
    First Bone or Socket Index:NiagaraInt32=&MergeHelper_IntSwitch(0=2, 1=0, 2=2, 3=2, 4=?, 5=?, Index=$User.SwingSectionIndex, Num=::NewEnumerator2)
    Second Bone or Socket Index:NiagaraInt32=&MergeHelper_IntSwitch(0=3, 1=3, 2=1, 3=1, 4=?, 5=?, Index=$User.SwingSectionIndex, Num=::NewEnumerator2)
   )
   &FN_InitializeAnimTrail(
    AnimTrailWidth=0.2
    Color=&MultiplyLinearColor(
     Linear Color=$System.TOD
     Scale Factor=&MakeFloat4FromVectorAndFloat(
      "Vector (RGB)"=&MakeVectorFromLinearColorRGB(Color=<rgba 0.09 0.3 0.3 1>)
      "Float (Alpha)"=&Multiply_Float(A=$System.TrailFade, B=3)
     )
    )
    Lifetime=0.18
   )
  ]
  particleUpdateScript [
   &KillParticles(Kill Particles=$System.Kill)
   &ScaleColor(
    Scale Alpha=&FloatFromCurve(FloatCurve=DataInterface() {propertyValues {Curve {keys [
         {arriveTangent=-4.5312, interpMode=RCIM_Cubic, leaveTangent=-4.5312, tangentMode=RCTM_User, time=0.3, value=1}
         {interpMode=RCIM_Cubic, tangentMode=RCTM_User, time=1, value=0}
        ]}}})
    Scale RGB=&Multiply_Vector(
     A=&MakeVectorFromLinearColorRGB(Color=&ColorFromCurve(ColorCurve=DataInterface() {propertyValues {
         AlphaCurve {keys [<at 0 1 RCIM_Cubic>]}
         BlueCurve {keys [<at 0.4018 1>]}
         GreenCurve {keys [<at 0.4018 1>]}
         RedCurve {keys [<at 0.4018 1>]}
        }}))
     B=&MakeVectorFromLinearColorRGB(Color=$System.TOD)
    )
    ScaleA=true
    ScaleRGB=true
   )
   &FN_UpdateAnimTrail() {module=FN_UpdateAnimTrail001}
   &SolveForcesAndVelocity(Clamp Velocity=false, Limit Acceleration=false)
  ]
  rendererClasses [&NiagaraRibbonRendererProperties]
  renderers [{rendererClass=&NiagaraRibbonRendererProperties}]
 }
}
summary {
 systemName=NS_Pickaxe_CrawlyTech_Swing
 emitters [
  {emitterName=MainTrail, rendererClasses [&NiagaraRibbonRendererProperties]}
  {emitterName=TipTrail, rendererClasses [&NiagaraRibbonRendererProperties]}
  {emitterName=LightningTrail, rendererClasses [&NiagaraRibbonRendererProperties]}
 ]
 userVariables [
  User.bIsFrontEnd {type=NiagaraBool}
  User.bIsFrontEndPreview {type=NiagaraBool}
  User.RibbonIDOffset {type=NiagaraInt32}
  User.SwingSectionIndex {type=NiagaraInt32}
 ]
}
