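# Physics asset for the CrawlyTech pickaxe mesh, defining sphere bodies for the two-handed handle and a three-bone dyn_axis chain.
# One large sphere covers the handle, and each dyn_axis bone has a small sphere.
# Constraints link the bones in series from the handle, locking twist and swing1 and limiting swing2 to a narrow angle, so each joint bends slightly on a single axis.
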PhysicsAsset /BRCosmetics/Weapons/FORT_Melee/Pickaxe_CrawlyTech/Meshes/CrawlyTech_Axe_Physics
bodies {
 Melee_TwoHanded_Handle [{shapeType=Sphere, radius=20.7689, center=<-0.6594 -0.2126 5.5921>}]
 dyn_axis_1 [{shapeType=Sphere, radius=5, shapeName=dyn_axis_1_sphere}]
 dyn_axis_2 [{shapeType=Sphere, radius=5, shapeName=dyn_axis_2_sphere}]
 dyn_axis_3 [{shapeType=Sphere, radius=5, shapeName=dyn_axis_3_sphere}]
}
constraints [
 {bone1Name=dyn_axis_1, bone2Name=Melee_TwoHanded_Handle, swing1LimitDegrees=0, swing2LimitDegrees=10, twistLimitDegrees=0}
 {bone1Name=dyn_axis_2, bone2Name=dyn_axis_1, swing1LimitDegrees=0, swing2LimitDegrees=10, twistLimitDegrees=0}
 {bone1Name=dyn_axis_3, bone2Name=dyn_axis_2, swing1LimitDegrees=0, swing2LimitDegrees=10, twistLimitDegrees=0}
]
