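# Post-process style animation blueprint for the CrawlyTech pickaxe, adding secondary physics motion to the incoming pose.
# The cached input pose feeds two branches, selected by bIsInFrontEnd.
# In the front end a RigidBody node simulates with the axe physics asset; otherwise chained AnimDynamics nodes spring dyn_axis_1 to 3 within tight angular limits.
# A second, world-space AnimDynamics chain is left unconnected, and the event graph holds no logic.
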AnimBlueprint /BRCosmetics/Weapons/FORT_Melee/Pickaxe_CrawlyTech/Meshes/CrawlyTech_Axe_AnimBP
target_skeleton=&/CrawlyTech_Axe_Skeleton
graphs [
 AnimGraph {nodes [
   n1 = Anim Dynamics - Bone: dyn_axis_4(Alpha=1, ExternalForce="0,0,0") {
    angularBiasOverride=1
    angularDampingOverride=1
    angularSpringConstant=1000
    bAngularSpring=true
    bGravityOverrideInSimSpace=true
    bOverrideAngularBias=true
    bOverrideAngularDamping=true
    bUseGravityOverride=true
    simulationSpace=World
    alphaBoolBlend {blendOption=Linear}
    boundBone {boneName=dyn_axis_4}
    componentAppliedLinearAccClamp=<3000 3000 3000>
    componentLinearAccScale=<0.1 0.1 0.1>
    componentLinearVelScale=<0.1 0.1 0.1>
    gravityOverride {z=980}
    physicsBodyDefinitions [{
      boundBone {boneName=dyn_axis_4}
      boxExtents=<20 20 20>
      constraintSetup {angularLimitsMax=<0 8 0>, angularLimitsMin=<0 -8 0>, angularTarget=<0 0 0>}
      localJointOffset=<1 0 0>
     }]
   }
   n2 = Anim Dynamics - Bone: dyn_axis_3(ComponentPose=@n1, Alpha=1, ExternalForce="0,0,0") {
    angularBiasOverride=1
    angularDampingOverride=1
    angularSpringConstant=1000
    bAngularSpring=true
    bGravityOverrideInSimSpace=true
    bOverrideAngularBias=true
    bOverrideAngularDamping=true
    bUseGravityOverride=true
    simulationSpace=World
    alphaBoolBlend {blendOption=Linear}
    boundBone {boneName=dyn_axis_3}
    componentAppliedLinearAccClamp=<3000 3000 3000>
    componentLinearAccScale=<0.1 0.1 0.1>
    componentLinearVelScale=<0.1 0.1 0.1>
    gravityOverride {z=980}
    physicsBodyDefinitions [{
      boundBone {boneName=dyn_axis_3}
      boxExtents=<20 20 20>
      constraintSetup {angularLimitsMax=<0 8 0>, angularLimitsMin=<0 -8 0>, angularTarget=<0 0 0>}
      localJointOffset=<1 0 0>
     }]
   }
   n3 = Anim Dynamics - Bone: dyn_axis_2(ComponentPose=@n2, Alpha=1, ExternalForce="0,0,0") {
    angularBiasOverride=1
    angularDampingOverride=1
    angularSpringConstant=1000
    bAngularSpring=true
    bGravityOverrideInSimSpace=true
    bOverrideAngularBias=true
    bOverrideAngularDamping=true
    bUseGravityOverride=true
    simulationSpace=World
    alphaBoolBlend {blendOption=Linear}
    boundBone {boneName=dyn_axis_2}
    componentAppliedLinearAccClamp=<3000 3000 3000>
    componentLinearAccScale=<0.1 0.1 0.1>
    componentLinearVelScale=<0.1 0.1 0.1>
    gravityOverride {z=980}
    physicsBodyDefinitions [{
      boundBone {boneName=dyn_axis_2}
      boxExtents=<20 20 20>
      constraintSetup {angularLimitsMax=<0 2.5 0>, angularLimitsMin=<0 -2.5 0>, angularTarget=<0 0 0>}
      localJointOffset=<1 0 0>
     }]
   }
   Anim Dynamics - Bone: dyn_axis_1(ComponentPose=@n3, Alpha=1, ExternalForce="0,0,0") {
    angularBiasOverride=1
    angularDampingOverride=1
    angularSpringConstant=1000
    bAngularSpring=true
    bGravityOverrideInSimSpace=true
    bOverrideAngularBias=true
    bOverrideAngularDamping=true
    bUseGravityOverride=true
    simulationSpace=World
    alphaBoolBlend {blendOption=Linear}
    boundBone {boneName=dyn_axis_1}
    componentAppliedLinearAccClamp=<3000 3000 3000>
    componentLinearAccScale=<0.1 0.1 0.1>
    componentLinearVelScale=<0.1 0.1 0.1>
    gravityOverride {z=980}
    physicsBodyDefinitions [{
      boundBone {boneName=dyn_axis_1}
      boxExtents=<20 20 20>
      constraintSetup {angularLimitsMax=<0 2.5 0>, angularLimitsMin=<0 -2.5 0>, angularTarget=<0 0 0>}
      localJointOffset=<1 0 0>
     }]
   }
   n4 = Anim Dynamics - Bone: dyn_axis_1(ComponentPose=Local To Component(LocalPose=Use cached pose 'InputPoseCache'()), Alpha=1, ExternalForce="0,0,0") {
    angularBiasOverride=0.99
    angularDampingOverride=0.99
    angularSpringConstant=1000
    bAngularSpring=true
    bGravityOverrideInSimSpace=true
    bOverrideAngularBias=true
    bOverrideAngularDamping=true
    bUseGravityOverride=true
    alphaBoolBlend {blendOption=Linear}
    boundBone {boneName=dyn_axis_1}
    componentAppliedLinearAccClamp=<1000 1000 1000>
    componentLinearAccScale=<0.3 0.3 0.3>
    componentLinearVelScale=<0.3 0.3 0.3>
    gravityOverride {z=1500}
    physicsBodyDefinitions [{
      boundBone {boneName=dyn_axis_1}
      boxExtents=<25 25 25>
      constraintSetup {angularLimitsMax=<0 5 0>, angularLimitsMin=<0 -5 0>, angularTarget=<0 0 0>}
      localJointOffset=<0 0 1>
     }]
   }
   n5 = Anim Dynamics - Bone: dyn_axis_2(ComponentPose=@n4, Alpha=1, ExternalForce="0,0,0") {
    angularBiasOverride=0.99
    angularDampingOverride=0.99
    angularSpringConstant=1000
    bAngularSpring=true
    bGravityOverrideInSimSpace=true
    bOverrideAngularBias=true
    bOverrideAngularDamping=true
    bUseGravityOverride=true
    alphaBoolBlend {blendOption=Linear}
    boundBone {boneName=dyn_axis_2}
    componentAppliedLinearAccClamp=<1000 1000 1000>
    componentLinearAccScale=<0.3 0.3 0.3>
    componentLinearVelScale=<0.3 0.3 0.3>
    gravityOverride {z=1500}
    physicsBodyDefinitions [{
      boundBone {boneName=dyn_axis_2}
      boxExtents=<20 20 20>
      constraintSetup {angularLimitsMax=<0 10 0>, angularLimitsMin=<0 -10 0>, angularTarget=<0 0 0>}
      localJointOffset=<0 0 1>
     }]
   }
   n6 = Anim Dynamics - Bone: dyn_axis_3(ComponentPose=@n5, Alpha=1, ExternalForce="0,0,0") {
    angularBiasOverride=0.99
    angularDampingOverride=0.99
    angularSpringConstant=1000
    bAngularSpring=true
    bGravityOverrideInSimSpace=true
    bOverrideAngularBias=true
    bOverrideAngularDamping=true
    bUseGravityOverride=true
    alphaBoolBlend {blendOption=Linear}
    boundBone {boneName=dyn_axis_3}
    componentAppliedLinearAccClamp=<1000 1000 1000>
    componentLinearAccScale=<0.3 0.3 0.3>
    componentLinearVelScale=<0.3 0.3 0.3>
    gravityOverride {z=1500}
    physicsBodyDefinitions [{
      boundBone {boneName=dyn_axis_3}
      boxExtents=<15 15 15>
      constraintSetup {angularLimitsMax=<0 15 0>, angularLimitsMin=<0 -15 0>, angularTarget=<0 0 0>}
      localJointOffset=<0 0 1.5>
     }]
   }
   n7 = RigidBody(ComponentPose=Local To Component(LocalPose=Use cached pose 'InputPoseCache'())) {
    bOverrideWorldGravity=true
    overridePhysicsAsset=&/CrawlyTech_Axe_Physics
    alphaBoolBlend {blendOption=Linear}
    baseBoneRef {boneName=b_AR_Root}
    overrideWorldGravity {z=1500}
    simSpaceSettings {
     dampingAlpha=7
     maxAngularAcceleration=2500
     maxAngularVelocity=12
     maxLinearAcceleration=2500
     maxLinearVelocity=700
     worldAlpha=1
     externalLinearDragV=<0.5 0.5 0.5>
    }
   }
   Output Pose(Result=Blend Poses by bool(BlendPose_0=Component To Local(ComponentPose=@n7), BlendPose_1=Component To Local(ComponentPose=@n6), bActiveValue=Get bIsInFrontEnd()) {blendType=Linear}) {layerGroup=None}
   AnimGraphNode_SubInput_0 = Input Pose()
   Save cached pose 'InputPoseCache'(Pose=@AnimGraphNode_SubInput_0) {cacheName=InputPoseCache}
  ]}
 EventGraph {nodes [
   TryGetPawnOwner()
   Event BlueprintUpdateAnimation()
  ]}
]
variables []
